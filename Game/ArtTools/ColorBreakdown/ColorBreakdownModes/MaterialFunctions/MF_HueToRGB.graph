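# MF_HueToRGB is a library-exposed material function that converts a scalar Hue into a fully saturated RGB color, filed under ColorBreakdown art tools.
# It scales hue by six and builds each channel from an offset absolute-value ramp, then packs them with MakeFloat3 and saturates the result, avoiding branches.
# The Hue input previews with the U texture coordinate, and an RGBtoHSV call sits outside the output chain.
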MaterialFunction /Game/ArtTools/ColorBreakdown/ColorBreakdownModes/MaterialFunctions/MF_HueToRGB
graph {
 properties {bExposeToLibrary=true, libraryCategoriesText [Misc,Math]}
 nodes [
  n1 = TextureCoordinate()
  n2 = Multiply(A=FunctionInput(Preview=ComponentMask(None=@n1) {r=true}) {inputName=Hue, inputType=FunctionInput_Scalar}) {constB=6}
  n3 = Add(A=Abs(None=Add(A=@n2) {constB=-3})) {constB=-1}
  n4 = Subtract(B=Abs(None=Add(A=@n2) {constB=-2})) {constA=2}
  n5 = Subtract(B=Abs(None=Add(A=@n2) {constB=-4})) {constA=2}
  FunctionOutput(None=Saturate(None=MaterialFunctionCall(X=@n3, Y=@n4, Z=@n5) {materialFunction=&MakeFloat3}.Result))
  MaterialFunctionCall!(RGB) {materialFunction=/Engine/Functions/Engine_MaterialFunctions02/Math/RGBtoHSV}
 ]
}
