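# Material function for the ColorBreakdown materials that composites an overlay image on top of the incoming scene texture.
# Overlay UVs map viewport pixels to image texels, then apply offset and scale from the OverlayParams vector.
# A Step-based bounds mask combines with the image alpha to confine the overlay, avoiding branches.
# Named reroutes carry color and mask into nested lerps, with the scalar Overlay parameter controlling the final blend.
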MaterialFunction /Game/ArtTools/ColorBreakdown/ColorBreakdownModes/MaterialFunctions/MF_Overlay
graph {
 properties {description=Drops the overlay image on top of the input scene texture for the ColorBreakdown materials}
 nodes [
  n1 = FunctionInput() {inputName=SceneTextureIn}
  n2 = NamedRerouteUsage()
  n3 = NamedRerouteUsage()
  FunctionOutput(None=LinearInterpolate(A=@n1, B=LinearInterpolate(A=@n1, B=@n2, Alpha=@n3), Alpha=ScalarParameter() {parameterName=Overlay}))
  n4 = NamedRerouteUsage()
  n5 = Step(Y=@n4)
  n6 = Step(Y=@n4) {constX=0, constY=1}
  n7 = MaterialFunctionCall(Float2=Saturate(None=Subtract(A=@n5, B=@n6))) {materialFunction=&BreakOutFloat2Components}
  NamedRerouteDeclaration(None=Min(A=@n7.R, B=@n7.R)) {name=OverlayUVMask, nodeColor=<rgba 0.3359 1 0 1>}
  n8 = ViewSize()
  n9 = TextureObjectParameter() {parameterName=OverlayImage}
  n10 = TextureProperty(Tex=@n9)
  n11 = VectorParameter() {parameterName=OverlayParams, channelNames {a=OverlayYScale, b=OverlayXScale, g=OverlayV, r=OverlayU}, defaultValue=<rgba 0 0 2 2>}
  n12 = Subtract(
   A=Multiply(A=Divide(A=@n8, B=@n10), B=ScreenPosition().ViewportUV)
   B=Multiply(A=AppendVector(A=@n11.OverlayU, B=@n11.OverlayU), B=TextureProperty(Tex=@n9) {property=TMTM_TexelSize})
  )
  n13 = AppendVector(A=@n11.OverlayXScale, B=@n11.OverlayXScale)
  n14 = TextureSample(UVs=NamedRerouteDeclaration(None=Divide(A=@n12, B=@n13)) {name=OverlayUVs, nodeColor=<rgba 1 0 0.5859 1>}, Tex=@n9)
  NamedRerouteDeclaration(None=Multiply(A=@n14.A, B=NamedRerouteUsage())) {name=OverlayMask, nodeColor=<rgba 1 0 0.7969 1>}
  NamedRerouteDeclaration(None=@n14.RGB) {name=OverlayRGB, nodeColor=<rgba 0.2656 1 0 1>}
 ]
}
