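# Material function in the ColorBreakdown art tools that converts an RGB input to Hue, Chroma and Value, exposed to the library.
# Two If nodes order the channels into packed float4 intermediates P and Q, avoiding a per-channel hue branch.
# H and C are derived from those intermediates through named reroutes, then combined with Value into the float3 HCV output.
# The RGB input previews with a UV-driven hue gradient from MF_HueToRGB.
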MaterialFunction /Game/ArtTools/ColorBreakdown/ColorBreakdownModes/MaterialFunctions/MF_RGBtoHCV
graph {
 properties {bExposeToLibrary=true, description=Converts an input RGB value to Hue/Chroma/Value, libraryCategoriesText [Misc,Math]}
 nodes [
  n1 = MaterialFunctionCall(Float4=NamedRerouteUsage()) {materialFunction=&BreakOutFloat4Components}
  n2 = Subtract(A=@n1.A, B=@n1.A)
  n3 = Add(A=Multiply(B=NamedRerouteUsage()) {constA=6}) {constB=0}
  n4 = Abs(None=Add(A=Divide(A=@n2, B=@n3), B=@n1.B))
  n5 = MaterialFunctionCall(Float4=NamedRerouteUsage()) {materialFunction=&BreakOutFloat4Components}
  n6 = Subtract(A=@n5.R, B=Min(A=@n5.G, B=@n5.G))
  n7 = NamedRerouteUsage()
  FunctionOutput(None=MaterialFunctionCall(
    X=NamedRerouteDeclaration(None=@n4) {name=H, nodeColor=<rgba 1 0 0.8906 1>}
    Y=NamedRerouteDeclaration(None=@n6) {name=C, nodeColor=<rgba 1 0 0.0937 1>}
    Z=ComponentMask(None=@n7) {r=true}
   ) {materialFunction=&MakeFloat3}.Result) {outputName=HCV}
  n8 = ComponentMask(None=TextureCoordinate()) {r=true}
  n9 = MaterialFunctionCall(Float3=FunctionInput(Preview=MaterialFunctionCall!(Hue=@n8) {materialFunction=&/MF_HueToRGB}.Result) {inputName=RGB}) {materialFunction=&BreakOutFloat3Components}
  n10 = Constant()
  n11 = Divide() {constA=-1, constB=3}
  n12 = Constant() {r=-1}
  n13 = Divide() {constA=2, constB=3}
  NamedRerouteDeclaration(None=If(
    A=@n9.G
    B=@n9.G
    "A > B"=MaterialFunctionCall(X=@n9.G, Y=@n9.G, Z=@n10, A=@n11) {materialFunction=&MakeFloat4}.Result
    "A < B"=MaterialFunctionCall(X=@n9.B, Y=@n9.B, Z=@n12, A=@n13) {materialFunction=&MakeFloat4}.Result
   )) {name=P, nodeColor=<rgba 0 0.8125 1 1>}
  n14 = NamedRerouteUsage()
  n15 = MaterialFunctionCall(Float4=NamedRerouteUsage()) {materialFunction=&BreakOutFloat4Components}
  n16 = MaterialFunctionCall(Float4=NamedRerouteUsage()) {materialFunction=&BreakOutFloat4Components}
  NamedRerouteDeclaration(None=If(
    A=@n9.R
    B=ComponentMask(None=@n14) {r=true}
    "A > B"=MaterialFunctionCall(X=@n9.R, Y=@n15.G, Z=@n15.G, A=@n15.G) {materialFunction=&MakeFloat4}.Result
    "A < B"=MaterialFunctionCall(X=@n16.R, Y=@n16.R, Z=@n16.R, A=@n9.R) {materialFunction=&MakeFloat4}.Result
   )) {name=Q, nodeColor=<rgba 0 0.8594 1 1>}
  NamedRerouteUsage()
 ]
}
