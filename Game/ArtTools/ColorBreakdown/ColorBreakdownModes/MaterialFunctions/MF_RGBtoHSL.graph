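# Library material function that converts an RGB input to HSL, part of the ColorBreakdown art tools.
# MF_RGBtoHCV supplies hue, chroma and value; lightness is value minus half the chroma.
# Saturation is chroma divided by 1 - |2L - 1|, with a zero-valued add leaving that divisor unguarded at black and white.
# The input previews with a hue ramp from MF_HueToRGB, and the output packs H, S and L into one float3.
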MaterialFunction /Game/ArtTools/ColorBreakdown/ColorBreakdownModes/MaterialFunctions/MF_RGBtoHSL
graph {
 properties {bExposeToLibrary=true, description=Converts an input RGB value to HSL, libraryCategoriesText [Misc,Math]}
 nodes [
  n1 = MaterialFunctionCall!(Hue=ComponentMask(None=TextureCoordinate()) {r=true}) {materialFunction=&/MF_HueToRGB}
  n2 = MaterialFunctionCall(Float3=MaterialFunctionCall!(RGB=FunctionInput(Preview=@n1.Result) {inputName=RGB}) {materialFunction=&/MF_RGBtoHCV}.HCV) {materialFunction=&BreakOutFloat3Components}
  n3 = Subtract(A=@n2.B, B=Multiply(A=@n2.G) {constB=0.5})
  n4 = Subtract(A=Multiply(A=@n3) {constB=2})
  n5 = OneMinus(None=Add(A=Abs(None=@n4)) {constB=0})
  FunctionOutput(None=MaterialFunctionCall(X=@n2.R, Y=Divide(A=@n2.G, B=@n5), Z=@n3) {materialFunction=&MakeFloat3}.Result)
 ]
}
